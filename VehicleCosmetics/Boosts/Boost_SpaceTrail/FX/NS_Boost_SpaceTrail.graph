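# Niagara system for the Space Trail vehicle boost cosmetic, drawing short-lived orange-gold ribbon trails from filtered car sockets.
# The one enabled emitter spawns a quality-scaled ribbon per port, offset on the skeletal mesh, with game-mode-selected backward velocity.
# Four ribbon renderers accompany three color layers with animated alpha, while spawn group flips ribbon width and sets per-layer widths.
# An older-style duplicate emitter and painted-variant overrides are disabled.
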NiagaraSystem /VehicleCosmetics/Boosts/Boost_SpaceTrail/FX/NS_Boost_SpaceTrail
dependencies {
 usedDataInterfaces [
  &NiagaraDataInterfaceCurve
  &NiagaraDataInterfaceColorCurve
  &NiagaraDataInterfaceArrayFloat
  &NiagaraDataInterfaceSkeletalMesh
 ]
 usedDynamicInputs []
 usedModules [
  &self:SystemSpawnScript.SystemScriptSource.SystemScriptGraph.NiagaraNodeAssignment_0.SetVariables_#1
  &SystemState
 ]
}
emitters {
 NE_VC_Boost_RibbonDuplicate {
  bEnabled=false
  emitterUpdateScript [
   &EmitterState()
   &VC_FilteredSocketCount()
   &VC_RateQualityHandler(High Spawn Rate=60, Low Spawn Rate=60, Medium Spawn Rate=60, Use Custom Rate?=true)
   &VC_SpawnRibbon_v1(Number of Ribbons=$Emitter.PortCount, SpawnRate=$Emitter.Spawn Rate)
  ]
  particleSpawnScript [
   &VC_PaintedVariantsOverride() {enabled=false}
   &VC_InitializeRibbon_v1(
    Custom Offset=true
    Enable Skeletal Position=true
    Lifetime=0.8
    Local Vector Offset=&AddVectorToPosition(Position=$User.DTCarOffset, Vector=<-40 0 0>)
    Offset Amount=1
    Ribbon Width=80
   )
   &AddVelocity(Velocity=&MergeHelper_VectorSwitch(0=<-1200 0 0>, 1=<-1500 0 0>, 2=<-1000 0 0>, 3=?, Index=$User.GameModeID, Num=::NewEnumerator1)) {module=AddVelocity001}
   &Color()
   /SetVariables(
    Particles.R0:LinearColor=<rgba 0.1981 0.0612 0 1>
    Particles.R1:LinearColor=<rgba 1 0.1119 0 1>
    Particles.R2:LinearColor=<rgba 0.4375 0.2754 0.0618 1>
   ) {isSetParameters=true, module=SetVariables_#2}
  ]
  particleUpdateScript [
   &ParticleState()
   &InheritSourceMovement(ApplyForce=false, ApplyPosition=true, ApplyPositionFalloff=false)
   &SolveForcesAndVelocity(Clamp Velocity=false, Limit Acceleration=false)
   &VC_UpdateRibbon_v1(Enable Skeletal Position=true)
   &ScaleColor(Scale RGB=&VectorFromFloat(Value=1), ScaleA=true, ScaleRGB=true)
   &ScaleRibbonWidth(Ribbon Width Scale=/VehicleCosmetics/FX/Niagara/Modules/GameMode/VC_SetFloatByGameModeID(
     SameValueForLockerAndBR?:NiagaraBool=false
     Locker Float:NiagaraFloat=&FloatFromCurve(FloatCurve=DataInterface() {propertyValues {Curve {keys [<at 0 0.6>,<at 1 0.6>]}}})
     Battle Royale Float:NiagaraFloat=&FloatFromCurve(FloatCurve=DataInterface() {propertyValues {Curve {keys [<at 0 0.6>,<at 1 0.6>]}}})
     Rocket Racing Float:NiagaraFloat=&FloatFromCurve(FloatCurve=DataInterface() {propertyValues {Curve {keys [<at 0 0.6>,<at 1 0.6>]}}})
     GameModeID:NiagaraInt32=$System.GameModeID
    )) {module=ScaleRibbonWidth001}
   /SetVariables(
    Particles.R0:LinearColor=&MultiplyLinearColor(
     Linear Color=<rgba 0.1981 0.0612 0 1>
     Scale Factor=&Vector4FromCurve(Vector4Curve=DataInterface() {propertyValues {
        AlphaCurve {keys [<at 0.0104 0.074>,<at 0.2065 1>,<at 0.9774 0>]}
        BlueCurve {keys [<at 0 1>,<at 1 1>,<at 1 1>]}
        GreenCurve {keys [<at 0 1>,<at 1 1>,<at 1 1>]}
        RedCurve {keys [<at 0 1>,<at 1 1>,<at 1 1>]}
       }})
    )
    Particles.R1:LinearColor=&MultiplyLinearColor(
     Linear Color=<rgba 1 0.1119 0 1>
     Scale Factor=&Vector4FromCurve(Vector4Curve=DataInterface() {propertyValues {
        AlphaCurve {keys [<at 0 0>,<at 0.1089 2>,<at 1 0>]}
        BlueCurve {keys [<at 0 1>,<at 1 1>]}
        GreenCurve {keys [<at 0 1>,<at 1 1>]}
        RedCurve {keys [<at 0 1>,<at 1 1>]}
       }})
    )
    Particles.R2:LinearColor=&MultiplyLinearColor(
     Linear Color=<rgba 0.4397 0.2747 0.0612 1>
     Scale Factor=&Vector4FromCurve(Vector4Curve=DataInterface() {propertyValues {
        AlphaCurve {keys [<at 0 -0.075>,<at 0.2 0.6>,<at 1 0>]}
        BlueCurve {keys [<at 0 1>,<at 1 1>]}
        GreenCurve {keys [<at 0 1>,<at 1 1>]}
        RedCurve {keys [<at 0 1>,<at 1 1>]}
       }})
    )
   ) {isSetParameters=true, module=SetVariables_#3}
  ]
  rendererClasses [&NiagaraRibbonRendererProperties]
  renderers [
   {rendererClass=&NiagaraRibbonRendererProperties}
   {rendererClass=&NiagaraRibbonRendererProperties, rendererIndex=1}
   {rendererClass=&NiagaraRibbonRendererProperties, rendererIndex=2}
  ]
 }
 NE_VC_Boost_RibbonDuplicate001 {
  emitterUpdateScript [
   &EmitterState()
   &VC_FilteredSocketCount()
   &VC_RateQualityHandler(High Spawn Rate=60, Low Spawn Rate=30, Use Custom Rate?=true)
   &VC_SpawnRibbon_v1(Number of Ribbons=$Emitter.PortCount, SpawnRate=$Emitter.Spawn Rate)
  ]
  particleSpawnScript [
   &VC_PaintedVariantsOverride() {enabled=false}
   &VC_InitializeRibbon_v1(
    Custom Offset=true
    Enable Skeletal Position=true
    Lifetime=0.8
    Local Vector Offset=&AddVectorToPosition(Position=$User.DTCarOffset, Vector=<-38 0 0>)
    Offset Amount=1
    Ribbon Width=80
   )
   &AddVelocity(Velocity=&MergeHelper_VectorSwitch(0=<-1200 0 0>, 1=<-1500 0 0>, 2=<-1000 0 0>, 3=?, Index=$User.GameModeID, Num=::NewEnumerator1)) {module=AddVelocity001}
   &Color()
   /SetVariables(
    Particles.R0:LinearColor=<rgba 0.1981 0.0612 0 1>
    Particles.R1:LinearColor=<rgba 1 0.1119 0 1>
    Particles.R2:LinearColor=<rgba 0.4375 0.2754 0.0618 1>
   ) {isSetParameters=true, module=SetVariables_#2}
  ]
  particleUpdateScript [
   &ParticleState()
   &InheritSourceMovement(ApplyForce=false, ApplyPosition=true, ApplyPositionFalloff=false)
   &SolveForcesAndVelocity(Clamp Velocity=false, Limit Acceleration=false)
   &VC_UpdateRibbon_v1(Enable Skeletal Position=true)
   &ScaleColor(Scale RGB=&VectorFromFloat(Value=1), ScaleA=true, ScaleRGB=true)
   &ScaleRibbonWidth(Ribbon Width Scale=&VC_SetFloatByGameModeID(
     SameValueForLockerAndBR?:NiagaraBool=false
     Locker Float:NiagaraFloat=&FloatFromCurve(FloatCurve=DataInterface() {propertyValues {Curve {keys [<at 0 0.6>,<at 1 0.6>]}}})
     Battle Royale Float:NiagaraFloat=&FloatFromCurve(FloatCurve=DataInterface() {propertyValues {Curve {keys [<at 0 0.6>,<at 1 0.6>]}}})
     Rocket Racing Float:NiagaraFloat=&FloatFromCurve(FloatCurve=DataInterface() {propertyValues {Curve {keys [<at 0 0.6>,<at 1 0.6>]}}})
     GameModeID:NiagaraInt32=$System.GameModeID
    )) {module=ScaleRibbonWidth001}
   /SetVariables(
    Particles.R0:LinearColor=&MultiplyLinearColor(
     Linear Color=<rgba 0.1981 0.0612 0 1>
     Scale Factor=&Vector4FromCurve(Vector4Curve=DataInterface() {propertyValues {
        AlphaCurve {keys [<at 0.0104 0.074>,<at 0.0936 0.85>,<at 0.5055 0.75>,<at 0.7916 0.5>]}
        BlueCurve {keys [<at 0 1>,<at 1 1>,<at 1 1>]}
        GreenCurve {keys [<at 0 1>,<at 1 1>,<at 1 1>]}
        RedCurve {keys [<at 0 1>,<at 1 1>,<at 1 1>]}
       }})
    )
    Particles.R1:LinearColor=&MultiplyLinearColor(
     Linear Color=<rgba 1.2 0.1343 0 1>
     Scale Factor=&Vector4FromCurve(Vector4Curve=DataInterface() {propertyValues {
        AlphaCurve {keys [<at 0.0006 0>,<at 0.2078 0.95>,<at 0.5314 0.95>,<at 1 0>]}
        BlueCurve {keys [<at 0 1>,<at 1 1>]}
        GreenCurve {keys [<at 0 1>,<at 1 1>]}
        RedCurve {keys [<at 0 1>,<at 1 1>]}
       }})
    )
    Particles.R2:LinearColor=&MultiplyLinearColor(
     Linear Color=<rgba 0.35 0.2187 0.0488 0.95>
     Scale Factor=&Vector4FromCurve(Vector4Curve=DataInterface() {propertyValues {
        AlphaCurve {keys [<at 0 0.002>,<at 0.0895 0.45>,<at 0.6975 0.2>,<at 0.918 0>]}
        BlueCurve {keys [<at 0 1>,<at 1 1>]}
        GreenCurve {keys [<at 0 1>,<at 1 1>]}
        RedCurve {keys [<at 0 1>,<at 1 1>]}
       }})
    )
   ) {isSetParameters=true, module=SetVariables_#3}
   /SetVariables(
    Particles.R1Width:NiagaraFloat=&SelectFloatFromArray(Direct Array Index=&SpawnGroup(), Float Selection Array=DataInterface() {propertyValues {FloatData [0,-50]}})
    Particles.R2Width:NiagaraFloat=&SelectFloatFromArray(Direct Array Index=&SpawnGroup(), Float Selection Array=DataInterface() {propertyValues {FloatData [50,0]}})
   ) {isSetParameters=true}
   &ScaleRibbonWidth(Ribbon Width Scale=&SelectFloatFromArray(Direct Array Index=&SpawnGroup(), Float Selection Array=DataInterface() {propertyValues {FloatData [1,-1]}}))
  ]
  rendererClasses [&NiagaraRibbonRendererProperties]
  renderers [
   {rendererClass=&NiagaraRibbonRendererProperties}
   {rendererClass=&NiagaraRibbonRendererProperties, rendererIndex=1}
   {rendererClass=&NiagaraRibbonRendererProperties, rendererIndex=2}
   {rendererClass=&NiagaraRibbonRendererProperties, rendererIndex=3}
  ]
 }
}
summary {
 systemName=NS_Boost_SpaceTrail
 emitters [
  {bEnabled=false, emitterName=NE_VC_Boost_RibbonDuplicate, rendererClasses [&NiagaraRibbonRendererProperties]}
  {emitterName=NE_VC_Boost_RibbonDuplicate001, rendererClasses [&NiagaraRibbonRendererProperties]}
 ]
 userVariables [
  User.bIsFrontendPreview {type=NiagaraBool, defaultValue=true}
  User.bIsPainted {type=NiagaraBool}
  User.ColorVariantID {type=NiagaraInt32}
  User.DTCarOffset {type=NiagaraPosition, description="", defaultValue {}}
  User.PortCount {type=NiagaraInt32, defaultValue=2}
  User.DTCarOffset001 {type=Vector3f, description="", defaultValue {struct=Vector3f, value=<0 0 0>}}
  User.GameModeID {type=NiagaraInt32}
  User.PaintedColor {type=LinearColor, description="", defaultValue {struct=LinearColor, value=<rgba 1 1 1 1>}}
  User.PaintedScalar {type=NiagaraFloat, defaultValue=1, description=""}
 ]
}
